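# Additive, unlit, two-sided mesh-particle material for a hollow wave effect in the flintlock melee weapon's effects.
# Squared view facing and vertex red soften silhouettes and mask a turbulence texture panned by dynamic-parameter Time.
# That mask blends grey into particle color for emissive and drives opacity, with depth fade and adjustable top/bottom UV gradient fades.
# Optional slight refraction follows the edge mask and a dynamic parameter, weakening with camera distance.
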Material /Game/Weapons/FORT_Melee/Effects/FlintlockMelee/M_HollowWave
graph {
 properties {
  bUsedWithMeshParticles=true
  bUsedWithNiagaraMeshParticles=true
  blendMode=BLEND_Additive
  pixelDepthOffsetMode=PDOM_Legacy
  refractionMethod=RM_IndexOfRefraction
  shadingModel=MSM_Unlit
  translucencyPass=MTP_BeforeDOF
  twoSided=true
  usedShadingModels=Unlit
 }
 nodes [
  Composite() {subgraphName=Top/bottom Fade}
  n1 = DynamicParameter() {paramNames [Time,Random,DepthFade,Ref]}
  n2 = Abs(None=DotProduct(A=VertexNormalWS(), B=CameraVectorWS()))
  n3 = Multiply(A=VertexColor().R, B=Multiply(A=@n2, B=@n2))
  n4 = Divide(B=Distance(A=ObjectPositionWS(), B=CameraPositionWS())) {constA=350}
  n5 = LinearInterpolate(Alpha=Multiply(A=Multiply(A=@n1.Ref, B=@n3), B=Clamp(None=@n4))) {constA=1, constB=1.05}
  LinearInterpolate(B=FeatureLevelSwitch(Default=@n5, SM4_REMOVED=@n5), Alpha=ScalarParameter() {defaultValue=1, parameterName=Use Refraction}) {constA=1}
  n6 = ParticleColor()
  n7 = TextureCoordinate() {vTiling=0.1}
  n8 = Add(A=@n1.Time, B=@n1.Time)
  n9 = Multiply(
   A=@n3
   B=TextureSampleParameter2D(UVs=Panner(Coordinate=@n7, Time=@n8) {speedY=0.5}) {parameterName=Wave Texture, texture=/Game/Global/Textures/Default/Caustics/T_Turbulence_Seamless}.R
  )
  n10 = Multiply(A=LinearInterpolate(A=VectorParameter() {defaultValue=<rgba 0.4948 0.4948 0.4948 1>}.RGB, B=@n6.RGB, Alpha=@n9), B=@n6.A)
  MaterialFunctionCall(Emissive=LinearInterpolate(A=@n10, B=Multiply(A=@n6.RGB, B=@n10), Alpha=ScalarParameter() {parameterName=Use Full Particle Color})) {materialFunction=&MF_FX_EmissiveAdjustments}
  Multiply()
  n11 = LinearInterpolate(A=@n9, B=Multiply(A=@n6.A, B=@n9), Alpha=ScalarParameter() {parameterName=Use Full Particle Color})
  n12 = MaterialFunctionCall() {materialFunction=&LinearGradient}
  n13 = OneMinus(None=@n12.VGradient)
  n14 = VectorParameter() {parameterName=Bottom Fade, channelNames {a=contrast, b=add, g=mul, r=pow}, defaultValue=<rgba 1 1 1 0>}
  n15 = Add(A=Multiply(A=Power(Base=@n13, Exp=@n14.pow), B=@n14.mul), B=@n14.add)
  n16 = VectorParameter() {parameterName=Top Fade, channelNames {a=contrast, b=add, g=mul, r=pow}, defaultValue=<rgba 1 1 1 0>}
  n17 = Add(A=Multiply(A=Power(Base=@n12.VGradient, Exp=@n16.pow), B=@n16.mul), B=@n16.add)
  n18 = Min(
   A=Saturate(None=MaterialFunctionCall(In=@n15, Contrast=@n14.contrast) {materialFunction=&CheapContrast}.Result)
   B=Saturate(None=MaterialFunctionCall(In=@n17, Contrast=@n16.contrast) {materialFunction=&CheapContrast}.Result)
  )
  MaterialExpressionReroute_0 = Reroute(None=@n18)
  Multiply(
   A=ShadingPathSwitch(Default=DepthFade(Opacity=@n11, FadeDistance=@n1.DepthFade), Forward=@n9, Mobile=@n9)
   B=@MaterialExpressionReroute_0.Output
  )
  PinBase()
  PinBase!(Output=@n18) {reroutePins [Output {expression=&self:MaterialExpressionReroute_0}]}
 ]
}
